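# Translucent, lit particle material for the seed effect in the Aspire gadget emote, flagged for Niagara sprites as well as skeletal meshes.
# The red channel of a side-view seed texture acts as a mask, multiplied separately by particle color RGB and particle alpha.
# This lets the particle system control the tint and fade of each seed.
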Material /Game/Animation/Props/Gadgets_Aspire/FX/M_Aspire_Emote_Seeds
graph {
 properties {
  bUsedWithMeshDeformer=true
  bUsedWithNiagaraSprites=true
  bUsedWithSkeletalMesh=true
  blendMode=BLEND_Translucent
  pixelDepthOffsetMode=PDOM_Legacy
  usedShadingModels=Default Lit
 }
 nodes [
  n1 = TextureSample() {texture=&/T_Aspire_Seed_Side}
  n2 = ParticleColor()
  Multiply(A=@n1.R, B=@n2.RGB)
  Multiply(A=@n1.R, B=@n2.A)
 ]
}
